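; Weapon blueprint for the Shield Flopper consumable, handling a drip particle effect on the held fish.
; When equipped and attached, it calls the parent handler, waits briefly, then activates the P_Flopper_Catch_Drips component with a reset.
; When unequipped and detached, it calls the parent and deactivates that component.
; BeginPlay and Tick are unused stubs, and the construction script only calls its parent.
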(event EventBeginPlay
  (|CallParentFunction) ; This node is disabled and will not be called.\nDrag off pins to build functionality.
)

(event EventTick
  (params (DeltaSeconds :type Float%20%28single-precision%29))
  (|CallParentFunction DeltaSeconds) ; This node is disabled and will not be called.\nDrag off pins to build functionality.
)

(event FortWeapon|EventOnEquippedAndAttached
  (|CallParentFunction)
  (Utilities|FlowControl|Delay 0.2)
  (Components|Activation|Activate (Variables|Default|GetP_Flopper_Catch_Drips) true))

(event FortWeapon|EventOnUnEquippedAndDetached
  (|CallParentFunction)
  (Components|Activation|Deactivate (Variables|Default|GetP_Flopper_Catch_Drips)))

(fn ConstructionScript
  (|CallParentFunction))
